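# Niagara swing trail effect for the BrightGunnerRemix1H one-handed pickaxe, pairing an animation-driven ribbon with sparks.
# NE_Trail_01 spawns ribbon particles between two bone or socket indices, with short life, dynamic material parameters, and alpha and width curves.
# FX_Sparks spawns small sprites by distance moved, which drift, fall and shrink, flashing very bright before dimming.
# Spark alpha fades over life and is scaled by time of day, lower at night.
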NiagaraSystem /BRCosmetics/Weapons/FORT_Melee/Pickaxe_BrightGunnerRemix1H/FX/NS_Pickaxe_BrightGunnerRemix1H_Trail
dependencies {
 usedDataInterfaces [
  &NiagaraDataInterfaceCurve
  &NiagaraDataInterfaceVectorCurve
  &NiagaraDataInterfaceSocketReader
  &NiagaraDataInterfaceColorCurve
 ]
 usedDynamicInputs []
 usedModules [
  &SystemState
  &self:SystemSpawnScript.SystemScriptSource.SystemScriptGraph.NiagaraNodeAssignment_1.SetVariables_#1
  &self:SystemSpawnScript.SystemScriptSource.SystemScriptGraph.NiagaraNodeAssignment_2.SetVariables_#2
 ]
}
emitters {
 FX_Sparks {
  emitterUpdateScript [
   &EmitterState()
   &SpawnPerUnit(
    Module.Use Max Movement Threshold=true
    Module.Use Movement Tolerance=true
    Movement Tolerance=0.1
    Spawn Spacing=25
    Use Spawn Probability=false
   )
  ]
  particleSpawnScript [
   &InitializeParticle(
    Color Maximum=<rgba 12 12 12 0.5>
    Color Minimum=<rgba 0 0 0 0.25>
    Color Mode=::NewEnumerator2
    Lifetime Max=0.85
    Lifetime Min=0.05
    Lifetime Mode=::NewEnumerator1
    Sprite Rotation Mode=::NewEnumerator2
    Sprite Size Mode=::NewEnumerator1
    Uniform Sprite Size Max=6
    Uniform Sprite Size Min=0.1
    UsePositionOffset=false
    Write Material Random=true
   )
   &ShapeLocation(MaskSpawnGroup=false, Override Seed=false, Sphere Radius=2, UseOwnerScale=true)
   &AddVelocity(Velocity=&RandomRangeVector(Minimum=<-4 0 0>, Maximum=<-35 0 0>, Evaluation Type=::NewEnumerator1))
  ]
  particleUpdateScript [
   &ParticleState()
   &SpriteRotationRate(Rotation Rate=&RandomRangeFloat(Minimum=-0.125, Maximum=0.125))
   &ScaleSpriteSize(
    Scale Sprite Size Mode=::NewEnumerator0
    Uniform Scale Factor=&FloatFromCurve(FloatCurve=DataInterface() {propertyValues {Curve {keys [
         {arriveTangent=71.0578, interpMode=RCIM_Cubic, leaveTangent=71.0578, tangentMode=RCTM_User, time=0, value=0.0033}
         {arriveTangent=10.4834, interpMode=RCIM_Cubic, leaveTangent=10.4834, tangentMode=RCTM_User, time=0.0168, value=0.9997}
         {arriveTangent=-0.0068, interpMode=RCIM_Cubic, leaveTangent=-0.0068, time=0.1127, value=1.4144}
         {arriveTangent=-0.1707, interpMode=RCIM_Cubic, leaveTangent=-0.1707, tangentMode=RCTM_User, time=0.3584, value=0.9974}
         {arriveTangent=-1.5546, interpMode=RCIM_Cubic, leaveTangent=-1.5546, time=0.8047, value=0.6633}
         <at 1 0>
        ]}}})
   )
   &ScaleColor(
    Scale Alpha=&FloatFromCurve(
     FloatCurve=DataInterface() {propertyValues {Curve {keys [<at 0 1>,<at 1 0>]}}}
     Scale Curve=&FNTimeOfDayScalar(Night Scale=0.1, Morning Scale=1, Evening Scale=1)
    )
    Scale RGB=&VectorFromCurve(VectorCurve=DataInterface() {propertyValues {XCurve {keys [<at 0 100>,<at 0.25 1>]}, YCurve {keys [<at 0 100>,<at 0.25 1>]}, ZCurve {keys [<at 0 100>,<at 0.25 1>]}}})
    ScaleA=true
    ScaleRGB=true
   )
   &AccelerationForce(Acceleration=<0 0 -80>)
   &SolveForcesAndVelocity(Clamp Velocity=false, Limit Acceleration=false)
  ]
  rendererClasses [&NiagaraSpriteRendererProperties]
  renderers [{rendererClass=&NiagaraSpriteRendererProperties}]
 }
 NE_Trail_01 {
  emitterSpawnScript [&FN_RibbonIDOffsetUtility()]
  emitterUpdateScript [
   &EmitterState()
   &FN_SpawnAnimTrail()
  ]
  particleSpawnScript [
   &FN_InitializeAnimTrailBoneIndecies(First Bone or Socket Index=1, Second Bone or Socket Index=0)
   &FN_InitializeAnimTrail(Lifetime=0.175)
  ]
  particleUpdateScript [
   &SolveForcesAndVelocity(Clamp Velocity=false, Limit Acceleration=false)
   &FN_UpdateAnimTrail(Lock Particles=true) {module=FN_UpdateAnimTrail001}
   &DynamicMaterialParameters(
    Index 1 Param 3=&FloatFromCurve(
     FloatCurve=DataInterface() {propertyValues {Curve {keys [<at 0 0>, {arriveTangent=1.9814, tangentMode=RCTM_User, time=1, value=1}]}}}
     CurveIndex=&ModuloFloat(A=$Emitter.Age, B=1)
     Scale Curve=0
    )
    Index 1 Param 4=1
    Index 2 Param 1=10
    Index 2 Param 2=1
    Index 2 Param 3=0.5
    Index 2 Param 4=0.6
    Module.Index0Float0Write=true
    Module.Index0Float1Write=true
    Module.Index0Float2Write=true
    Module.Index0Float3Write=true
    Module.Index1Float0Write=true
    Module.Index1Float1Write=true
    Module.Index1Float2Write=true
    Module.Index1Float3Write=true
    Module.Index2Float0Write=true
    Module.Index2Float1Write=true
    Module.Index2Float2Write=true
    Module.Index2Float3Write=true
    Write Parameter Index 1=::NewEnumerator1
    Write Parameter Index 2=::NewEnumerator1
   )
   &ScaleColor(
    Scale Alpha=&FloatFromCurve(FloatCurve=DataInterface() {propertyValues {Curve {keys [
         <at -0.0006 0 RCIM_Cubic>
         {arriveTangent=-0.0636, interpMode=RCIM_Cubic, leaveTangent=-0.0636, tangentMode=RCTM_User, time=0.0319, value=1}
         {arriveTangent=-3.8711, interpMode=RCIM_Cubic, leaveTangent=-3.8711, time=0.0691, value=1}
         {arriveTangent=-1.4545, interpMode=RCIM_Cubic, leaveTangent=-1.4545, tangentMode=RCTM_User, time=0.1867, value=0.4007}
         <at 1 0>
        ]}}})
    Scale RGB=&VectorFromFloat(Value=&FloatFromCurve(FloatCurve=DataInterface() {propertyValues {Curve {keys [<at 0 1>,<at 1 1>]}}}))
    ScaleA=true
    ScaleRGB=true
   )
   &ScaleRibbonWidth(Ribbon Width Scale=&FloatFromCurve(FloatCurve=DataInterface() {propertyValues {Curve {keys [
         {arriveTangent=0.4985, time=0, value=0.5003}
         {arriveTangent=-0.0697, interpMode=RCIM_Cubic, leaveTangent=-0.0697, tangentMode=RCTM_Break, time=0.0308, value=0.9987}
         {arriveTangent=0.0333, interpMode=RCIM_Cubic, leaveTangent=0.1887, tangentMode=RCTM_Break, time=0.298, value=1.0006}
         <at 1 0>
        ]}}}))
  ]
  rendererClasses [&NiagaraRibbonRendererProperties]
  renderers [{rendererClass=&NiagaraRibbonRendererProperties}]
 }
}
summary {
 systemName=NS_Pickaxe_BrightGunnerRemix1H_Trail
 emitters [
  {emitterName=NE_Trail_01, rendererClasses [&NiagaraRibbonRendererProperties]}
  {emitterName=FX_Sparks, rendererClasses [&NiagaraSpriteRendererProperties]}
 ]
 userVariables [
  User.bIsFrontEnd {type=NiagaraBool}
  User.bIsFrontEndPreview {type=NiagaraBool}
  User.PhysicalSurface {type=NiagaraInt32}
  User.RibbonIDOffset {type=NiagaraInt32}
  User.SwingSectionIndex {type=NiagaraInt32}
 ]
}
